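# Sound cue for opening a porta-potty, layering a comedic fart over a metal door opening sound at half volume.
# A quality-level switch picks the mix: the first level randomly chooses among seven fart waves, while the other three share one mix with a single fixed fart.
# Each fart branch passes through a modulator.
# Attenuation and concurrency are overridden for short-range falloff and to block new instances at the limit.
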SoundCue /Game/Sounds/Fort_Test_Sounds/World_Objects/Plumbing/PortaPotty_Open_Cue
root=QualityLevel_0
nodes [
 {id=Mixer_0, children [Modulator_0,WavePlayer_0], properties {inputVolume [1,0.5]}}
 {id=Mixer_1, children [Modulator_1,WavePlayer_9], properties {inputVolume [1,0.5]}}
 {id=Modulator_0, children [Random_0]}
 {id=Modulator_1, children [WavePlayer_7]}
 {id=QualityLevel_0, children [Mixer_0,Mixer_1,Mixer_1,Mixer_1]}
 {
  id=Random_0
  children [WavePlayer_15, WavePlayer_16, WavePlayer_17, WavePlayer_18, WavePlayer_19, WavePlayer_20, WavePlayer_5]
  properties {weights [1,1,1,1,1,1,1]}
 }
 {id=WavePlayer_0, properties {soundWaveAssetPtr=/Game/Sounds/Fort_Doors/World/metal_door_open_B}}
 {id=WavePlayer_15, properties {soundWaveAssetPtr=&/HumanFart3_Mono}}
 {id=WavePlayer_16, properties {soundWaveAssetPtr=&/HumanFart2_Mono}}
 {id=WavePlayer_17, properties {soundWaveAssetPtr=&/HumanFart4_Mono}}
 {id=WavePlayer_18, properties {soundWaveAssetPtr=&/HumanFart5_Mono}}
 {id=WavePlayer_19, properties {soundWaveAssetPtr=&/HumanFart6_Mono}}
 {id=WavePlayer_20, properties {soundWaveAssetPtr=&/HumanFart7_Mono}}
 {id=WavePlayer_5, properties {soundWaveAssetPtr=&/HumanFart1}}
 {id=WavePlayer_7, properties {soundWaveAssetPtr=&HumanFart3_Mono}}
 {id=WavePlayer_9, properties {soundWaveAssetPtr=&metal_door_open_B}}
]
properties {
 bOverrideAttenuation=true
 bOverrideConcurrency=true
 maxDistance=600
 soundClassObject=&Fort_Loot_Collect
 volumeMultiplier=1
 attenuationOverrides {falloffDistance=400, attenuationShapeExtents {x=200}}
 concurrencyOverrides {resolutionRule=PreventNew}
}
